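# M_UI_BuzzCoolReticle is a translucent UI-domain material, named as a reticle for the BuzzCool UI.
# It feeds vertically offset texture coordinates and a fixed size into the MF_UI_Circle function, storing its Outline output in a CircleOutline named reroute.
# A second offset UV chain has no consumer, and no connection to the material's final outputs is shown, so it appears unfinished.
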Material /BuzzCool/UI/Materials/M_UI_BuzzCoolReticle
graph {
 properties {bAutomaticallySetUsageInEditor=false, blendMode=BLEND_Translucent, materialDomain=MD_UI, usedShadingModels=Default Lit}
 nodes [
  n1 = TextureCoordinate()
  AppendVector(A=ComponentMask(None=@n1) {r=true}, B=Add(A=ComponentMask(None=@n1) {g=true}) {constB=-0.25})
  n2 = TextureCoordinate()
  n3 = ComponentMask(None=@n2) {r=true}
  n4 = Add(A=ComponentMask(None=@n2) {g=true}) {constB=0.2}
  NamedRerouteDeclaration(None=MaterialFunctionCall(UVs=AppendVector(A=@n3, B=@n4), Size=Constant() {r=0.4}) {materialFunction=&MF_UI_Circle}.Outline) {name=CircleOutline, nodeColor=<rgba 0 0 0 1>}
  NamedRerouteUsage()
 ]
}
