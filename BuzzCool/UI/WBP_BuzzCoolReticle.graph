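# Reticle extension widget for the BuzzCool utility item that shows a radial cooldown indicator.
# On Construct, async tasks watch the owning controller's view target for the BuzzCool cooldown gameplay tag, showing or collapsing Image_Cooldown.
# A one-second material Percent animation is offset by elapsed fraction and rescaled to match cooldown duration.
# The reticle image stays collapsed here, and the equipped-weapon cast chain has no event driving it.
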WidgetBlueprint /BuzzCool/UI/WBP_BuzzCoolReticle
animations [Anim_Cooldown {
  binding_count=1
  length_seconds=1
  object=&self:Anim_Cooldown
  sequence {bindings [Image_Cooldown {
     id=#1
     possessed_class=Image
     tracks [{
       type=MovieSceneWidgetMaterialTrack
       sections [{type=MovieSceneParameterSection, end_frame=20, start_frame=0, channels [Percent {type=MovieSceneScriptingFloatChannel, keys [[0,0],[20,1]]}]}]
      }]
    }]}
 }]
graphs [
 EventGraph {nodes [
   n1 = Cast To B_ConsumableSmall_Athena!(
    execute
    Object=GetEquippedWeapon!(self:Fort Weapon Reticle Extension Widget Base Object Reference) -> (ReturnValue:Fort Weapon Object Reference)
   ) -> (then, CastFailed, AsB Consumable Small Athena:B Consumable Small Athena Object Reference) {is=title K2Node_DynamicCast}
   > Cast To B_UtilityItem_WildEstate_BuzzCool!(execute, Object=@n1) -> (then, CastFailed, AsB Utility Item Wild Estate Buzz Cool:B Utility Item Wild Estate Buzz Cool Object Reference) {is=title K2Node_DynamicCast}
   Event Construct()
   > n2 = Cast To FortPlayerController(Object=GetOwningPlayer())
   n3 = GetViewTarget!(self:Controller Object Reference=@n2) -> (ReturnValue:Actor Object Reference)
   > n4 = WaitGameplayTagAddToActor!(
    execute:Exec
    TargetActor:Actor Object Reference=@n3
    Tag:Gameplay Tag Structure=(TagName="Granted.Cooldown.WildEstate.BuzzCool")
    OnlyTriggerOnce:bool=false
   ) -> (then:Exec, AsyncTaskProxy:Ability Async Wait Gameplay Tag Added Object Reference, Added:Exec) {is=title K2Node_AsyncAction}
   > n5 = WaitGameplayTagRemoveFromActor!(
    execute:Exec
    TargetActor:Actor Object Reference=@n3
    Tag:Gameplay Tag Structure=(TagName="Granted.Cooldown.WildEstate.BuzzCool")
    OnlyTriggerOnce:bool=false
   ) -> (then:Exec, AsyncTaskProxy:Ability Async Wait Gameplay Tag Removed Object Reference, Removed:Exec) {is=title K2Node_AsyncAction}
   SetVisibility(execute=@n5.Removed, self=Get Image_Cooldown() -> (Image_Cooldown:Image Object Reference), InVisibility=Collapsed)
   SetCooldownAnimation!(execute=@n4.Added, self, InActor:Actor Object Reference=@n3) -> (then)
  ]}
 SetCooldownAnimation {nodes [
   SetCooldownAnimation!() -> (then, InActor:Actor Object Reference) {is=title K2Node_FunctionEntry}
   > SetVisibility(self=Get Image_Cooldown(), InVisibility=HitTestInvisible)
   > n1 = GetCooldownTimeRemainingAndDuration!(
    execute
    Actor:Actor Object Reference=Get InActor() -> (InActor:Actor Object Reference)
    CooldownTags:Gameplay Tag Container Structure=(GameplayTags=((TagName="Granted.Cooldown.WildEstate.BuzzCool")))
   ) -> (then, TimeRemaining:float, CooldownDuration:float)
   > PlayAnimation(
    InAnimation=Get Anim_Cooldown() -> (Anim_Cooldown:Widget Animation Object Reference)
    StartAtTime=SafeDivide(A=float - float(A=@n1.CooldownDuration, B=@n1), B=@n1.CooldownDuration)
    PlaybackSpeed=SafeDivide(A=1, B=@n1.CooldownDuration)
   )
  ]}
]
variables []
widgets {
 info {
  inheritedWidgetCount=1
  parentClass=/Script/FortniteUI.FortWeaponReticleExtensionWidgetBase
  rootWidgetClass=&Overlay
  widgetCount=4
 }
 widgets [
  Overlay_0 {widgetClassPath=&Overlay, widget=Overlay_0}
  Image_Reticle {
   widgetClassPath=&Image
   bIsVariable=true
   parent=Overlay_0
   widget=Image_Reticle
   properties {
    visibility=Collapsed
    brush {
     bIsDynamicallyLoaded=false
     drawAs=Image
     imageType=FullColor
     mirroring=NoMirror
     resourceName=None
     resourceObject=/BuzzCool/UI/Materials/M_UI_BuzzCoolReticle
     tiling=NoTile
     imageSize=<128 128>
     margin=<ltrb 0 0 0 0>
     outlineSettings {
      bUseBrushTransparency=false
      roundingType=HalfHeightRadius
      width=0
      color {colorUseRule=UseColor_Specified, specifiedColor=<rgba 0 0 0 0>}
      cornerRadii=<0 0 0 0>
     }
     tintColor {colorUseRule=UseColor_Specified, specifiedColor=<rgba 1 1 1 1>}
     uVRegion {bIsValid=false, max=<0 0>, min=<0 0>}
    }
   }
   slot {class=OverlaySlot}
  }
  Image_Cooldown {
   widgetClassPath=&Image
   bIsVariable=true
   parent=Overlay_0
   widget=Image_Cooldown
   properties {
    visibility=Collapsed
    brush {
     bIsDynamicallyLoaded=false
     drawAs=Image
     imageType=FullColor
     mirroring=NoMirror
     resourceName=None
     resourceObject=/BuzzCool/UI/Materials/M_UI_RadialProgress_Ammo
     tiling=NoTile
     imageSize=<110 110>
     margin=<ltrb 0 0 0 0>
     outlineSettings {
      bUseBrushTransparency=false
      roundingType=HalfHeightRadius
      width=0
      color {colorUseRule=UseColor_Specified, specifiedColor=<rgba 0 0 0 0>}
      cornerRadii=<0 0 0 0>
     }
     tintColor {colorUseRule=UseColor_Specified, specifiedColor=<rgba 1 1 1 1>}
     uVRegion {bIsValid=false, max=<0 0>, min=<0 0>}
    }
    renderTransform {angle=133, scale=<1 1>, shear=<0 0>, translation=<87 0>}
   }
   slot {class=OverlaySlot, properties {horizontalAlignment=HAlign_Fill, verticalAlignment=VAlign_Fill}}
  }
  TutorialHighlight {widgetClassPath=&TutorialHighlight}
 ]
}
